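; Parent NPC gameplay ability that makes an NPC flee danger while it carries the inside-danger tag, ending immediately otherwise.
; Activation clears any pending end-flee timer or task, then starts a randomly delayed timer that adds a loose flee-danger tag to trigger flee behavior.
; A tag-removal task then waits for danger to clear and ends the ability after another random delay.
; Ending removes the flee tag and clears both timers and the task.
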(event Ability|EventActivateAbility
  (bind _returnvalue (Ability|GetOwningActorfromActorInfo))
  (bind _waitinsidedangertagremovedtask (Variables|Default|GetWaitInsideDangerTagRemovedTask))
  ;[0.01803 0.07137 0.1354] Check that the owning actor is inside danger. If they aren't, immediately end the ability
  (Utilities|IsValid _returnvalue
    (:Is%20Valid
      (bind _returnvalue_2 (GameplayTags|HasMatchingGameplayTag(Message) _returnvalue (Variables|Default|GetInsideDangerTag)))
      (if _returnvalue_2
        (Utilities|FlowControl|Sequence
          (:then_0
            ;[0.01803 0.07137 0.1354] Clear any end flee timer/task that may be running from a previous activation of this ability
            (Utilities|Time|ClearandInvalidateTimerbyHandle (Variables|Default|GetRemoveFleeTagHandle))
            (Utilities|IsValid _waitinsidedangertagremovedtask
              (:Is%20Valid
                (GameplayTasks|EndTask _waitinsidedangertagremovedtask))))
          (:then_1
            ;[0.01803 0.07137 0.1354] Make sure the timer to start fleeing isn't already active from a previous activation of this ability
            (if (not (Utilities|Time|IsValidTimerHandle (Variables|Default|GetAddFleeTagHandle)))
              ;[0.0348 0.1354 0.03252] Start fleeing danger after a small delay
              (bind _returnvalue_3 (Utilities|Time|SetTimerbyEvent (delegate Custom|StartFlee) (Math|Random|RandomFloatinRange (Variables|Default|GetStartFleeMinDelay) (Variables|Default|GetStartFleeMaxDelay))))
              (Variables|Default|SetAddFleeTagHandle _returnvalue_3))))
        (else
          _)))
    (:Is%20Not%20Valid
      _))
  (Ability|EndAbility))

(event Custom|StartFlee
  (bind _returnvalue (Ability|GetOwningActorfromActorInfo))
  (Utilities|IsValid _returnvalue
    (:Is%20Valid
      ;[0.0348 0.1354 0.03252] Add the flee danger tag to the owner to trigger the flee danger behavior
      (Ability|GameplayEffect|AddLooseGameplayTags _returnvalue (GameplayTags|MakeGameplayTagContainerfromTag (Variables|Default|GetFleeDangerTag)))
      ;[0.3125 0.1436 0.0644] Wait for the owner to no longer be in danger, then end the ability after a delay to clear the flee danger status
      (bind _asynctaskproxy (Ability|Tasks|WaitGameplayTagRemove (Variables|Default|GetInsideDangerTag) 0 true)
        (:then
          (Variables|Default|SetWaitInsideDangerTagRemovedTask _asynctaskproxy))
        (:Removed
          (bind _returnvalue_2 (Utilities|Time|SetTimerbyEvent (EventDispatchers|CreateEvent :!event K2_EndAbility) (Math|Random|RandomFloatinRange (Variables|Default|GetEndFleeMinDelay) (Variables|Default|GetEndFleeMaxDelay))))
          (Variables|Default|SetRemoveFleeTagHandle _returnvalue_2))))
    (:Is%20Not%20Valid
      (Ability|EndAbility))))

(event Ability|EventOnEndAbility
  (params (bWasCancelled :type Boolean))
  (bind _returnvalue (Ability|GetOwningActorfromActorInfo))
  (bind _waitinsidedangertagremovedtask (Variables|Default|GetWaitInsideDangerTagRemovedTask))
  ;[0.3125 0.1436 0.0644] Make sure the owner no longer has the flee danger tag and clear any timers and tasks
  (Utilities|FlowControl|Sequence
    (:then_0
      (Utilities|IsValid _returnvalue
        (:Is%20Valid
          (Ability|GameplayEffect|RemoveLooseGameplayTags _returnvalue (GameplayTags|MakeGameplayTagContainerfromTag (Variables|Default|GetFleeDangerTag))))))
    (:then_1
      (Utilities|Time|ClearandInvalidateTimerbyHandle (Variables|Default|GetAddFleeTagHandle))
      (Utilities|Time|ClearandInvalidateTimerbyHandle (Variables|Default|GetRemoveFleeTagHandle))
      (Utilities|IsValid _waitinsidedangertagremovedtask
        (:Is%20Valid
          (GameplayTasks|EndTask _waitinsidedangertagremovedtask))))))
